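# Niagara effect for the RoyalAngst backpack that sheds tumbling, pastel-tinted VBucks mesh particles.
# User bools for landed and sprint/slide gate two spawn rates and adjust velocity, point force and inherited movement, while front-end modifiers alter scale and launch speed.
# Particles arc under gravity and drag, scale in and out, and meet custom sphere blocking volumes at three body offsets and the camera.
# The wind force is wired but multiplied to zero.
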NiagaraSystem /BRCosmetics/Accessories/FORT_Backpacks/Backpack_F_MED_RoyalAngst/FX/NS_Backpack_RoyalAngst
dependencies {
 usedDataInterfaces [
  &NiagaraDataInterfaceArrayColor
  &NiagaraDataInterfaceMeshRendererInfo
  &NiagaraDataInterfaceCurve
  &NiagaraDataInterfaceCamera
 ]
 usedDynamicInputs []
 usedModules [&SystemState]
}
emitters {VBucks {
  emitterUpdateScript [
   &EmitterState(Life Cycle Mode=::NewEnumerator1, Loop Behavior=::NewEnumerator1, Loop Duration Mode=::NewEnumerator1, UseLoopDelay=false)
   &SpawnRate(
    Spawn Probability=&MakeCustomFloatFromBool(Bool=$User.bPawnLanded, True Float=1, False Float=0)
    SpawnRate=65
    Use Spawn Probability=true
   ) {module=SpawnRate001}
   &SpawnRate(
    Spawn Probability=&MakeCustomFloatFromBool(Bool=$User.bPawnSprintOrSlide, True Float=1, False Float=0)
    SpawnGroup=1
    SpawnRate=85
    Use Spawn Probability=true
   ) {module=SpawnRate002}
   /SetVariables(
    Emitter.WindOscillateTime:NiagaraFloat=&Divide_Float(A=$Engine.Time, B=44)
    Emitter.WindRotateTime:NiagaraFloat=&Divide_Float(A=$Engine.Time, B=66)
   ) {isSetParameters=true}
   /SetVariables(
    Emitter.WindIntensity:NiagaraFloat=&Multiply_Float(A=&Sine(Normalized Angle=$Emitter.WindOscillateTime, Period=1, Scale=0, Bias=0), B=0)
    Emitter.WindVector:Vector3f=&MakeVector(
     X=&Add_Float(
      A=&Sine(Normalized Angle=$Emitter.WindRotateTime, Period=1, Scale=1, Bias=0)
      B=&Cosine(Normalized Angle=$Emitter.WindRotateTime, Period=1, Scale=1, Bias=0)
     )
     Y=&Subtract_Float(
      A=&Cosine(Normalized Angle=$Emitter.WindRotateTime, Period=1, Scale=1, Bias=0)
      B=&Sine(Normalized Angle=$Emitter.WindRotateTime, Period=1, Scale=1, Bias=0)
     )
     Z=0
    )
   ) {isSetParameters=true}
  ]
  particleSpawnScript [
   &InitializeParticle(
    Color=&SelectLinearColorFromArray(
     Array Sampling Mode=::NewEnumerator0
     Direct Array Index=?
     Color Selection Array=DataInterface() {propertyValues {ColorData [<rgba 0.9 0.81 0.225 1>, <rgba 1 0.45 0.5875 1>, <rgba 0.36 0.9 0.5535 1>, <rgba 0.8019 0.5 1 1>, <rgba 1 0.5127 0.3 1>]}}
     Random Seed=0
    )
    Lifetime Max=2.5
    Lifetime Min=1.8
    Lifetime Mode=::NewEnumerator1
    Mesh Scale Mode=::NewEnumerator3
    Mesh Uniform Scale=&FN_FrontEndFloatModifier(Gameplay=0.5, Front End=0.6)
    Position Offset=&RandomRangeVector(Minimum=<-8 -4 -1>, Maximum=<8 -8 0>)
    Position Offset Coordinate Space=::Local
    UsePositionOffset=true
    Write Material Random=true
   )
   &AddVelocity(
    Velocity=<0 -75 50>
    Velocity Speed Scale=&Multiply_Float(
     A=&FN_FrontEndFloatModifier(Gameplay=1, Front End=0.3)
     B=&RandomRangeFloat(Minimum=0.8, Maximum=1)
    )
   )
   &AddVelocity(
    Module.Use Falloff Distance=false
    Override Seed=false
    Velocity Mode=::NewEnumerator1
    Velocity Speed=&MakeCustomFloatFromBool(Bool=$User.bPawnSprintOrSlide, True Float=200, False Float=0)
   ) {module=AddVelocity001}
   &ApplyInitialForces()
   &InitialMeshOrientation(Use Rotation Vector=false)
  ]
  particleUpdateScript [
   &ParticleState()
   &ScaleMeshSize(Scale Factor=&VectorFromFloat(Value=&FloatFromCurve(
      FloatCurve=DataInterface() {propertyValues {Curve {keys [
          {arriveTangent=4.1785, interpMode=RCIM_Cubic, leaveTangent=8.7018, tangentMode=RCTM_Break, time=0, value=0}
          {arriveTangent=0.0224, interpMode=RCIM_Cubic, leaveTangent=0.0224, tangentMode=RCTM_Break, time=0.3, value=1}
         ]}}}
      CurveIndex=$Particles.Age
      Scale Curve=&FloatFromCurve(FloatCurve=DataInterface() {propertyValues {Curve {keys [
           {interpMode=RCIM_Cubic, tangentMode=RCTM_Break, time=0.7, value=1}
           {arriveTangent=-9.1079, interpMode=RCIM_Cubic, tangentMode=RCTM_Break, time=1, value=0}
          ]}}})
     )))
   &GravityForce(Gravity=<0 0 -1225>)
   &PointForce(
    Force Falloff Distance=333
    Force Strength=&MakeCustomFloatFromBool(Bool=$User.bPawnLanded, True Float=777, False Float=1500)
    Module.Use Falloff=true
    Module.UseLinearForce=false
   )
   &InheritSourceMovement(
    Applied Position Falloff Distance=500
    Applied Position Scale=&Multiply_VectorByFloat(
     Vector=&MakeCustomVectorFromBool(Bool=$User.bPawnLanded, True Vector=<0 0 0>, False Vector=<1 1 1>)
     Float=&FloatFromCurve(
      FloatCurve=DataInterface() {propertyValues {Curve {keys [
          {arriveTangent=-0.0029, interpMode=RCIM_Cubic, leaveTangent=-0.0029, tangentMode=RCTM_User, time=0, value=1}
          {interpMode=RCIM_Cubic, tangentMode=RCTM_User, time=1, value=0}
         ]}}}
      CurveIndex=$Particles.Age
     )
    )
    ApplyForce=false
    ApplyPosition=true
    ApplyPositionFalloff=false
   )
   &UpdateMeshOrientation(
    Rotation Rate=&RandomRangeFloat(
     Minimum=&RandomRangeFloat(Minimum=-1.25, Maximum=-1)
     Maximum=&RandomRangeFloat(Minimum=1.25, Maximum=1)
    )
    Rotation Vector=&Add_Vector(A=<1 1 1>, B=&RandomRangeVector(Minimum=<0.5 0 0>, Maximum=<1 0 1>))
   )
   &WindForce(Offset By Time:Vector3f=<0 0 100>, Wind Speed=$Emitter.WindVector, Wind Speed Scale=$Emitter.WindIntensity)
   &AerodynamicDrag(Aerodynamic Drag=0.5, Rotation Strength=2, Use Linear Drag=true, Use Rotational Drag=true)
   &SolveForcesAndVelocity(Clamp Velocity=false, Limit Acceleration=false)
   &AlignSpriteToMeshOrientation(SetAlignment=true, SetFacing=true)
   &self:SphereBlockingVolume_02(
    Pivot Offset:Vector3f=<0 20 -11>
    Sphere Origin:NiagaraPosition=$Engine.Emitter.SimulationPosition
    Sphere Radius:NiagaraFloat=&FN_FrontEndFloatModifier(Gameplay=25, Front End=20)
   ) {module=SphereBlockingVolume_02}
   &self:SphereBlockingVolume_02(Pivot Offset:Vector3f=<0 20 40>, Sphere Origin:NiagaraPosition=$Engine.Emitter.SimulationPosition, Sphere Radius=25) {module=SphereBlockingVolume_02001}
   &self:SphereBlockingVolume_02(Pivot Offset:Vector3f=<0 20 -75>, Sphere Origin:NiagaraPosition=$Engine.Emitter.SimulationPosition, Sphere Radius=25) {module=SphereBlockingVolume_02003}
   &self:SphereBlockingVolume_02(
    Sphere Origin:Vector3f=&ReturnCameraPosition(Coordinate Space:ENiagaraCoordinateSpace=::Simulation, Camera Query:NiagaraDataInterfaceCamera=DataInterface())
    Sphere Radius=66
   ) {module=SphereBlockingVolume_02002}
  ]
  rendererClasses [&NiagaraMeshRendererProperties]
  renderers [{rendererClass=&NiagaraMeshRendererProperties, rendererIndex=0}]
 }}
summary {
 systemName=NS_Backpack_RoyalAngst
 emitters [{emitterName=VBucks, rendererClasses [&NiagaraMeshRendererProperties]}]
 userVariables [
  User.bIsFrontEnd {type=NiagaraBool}
  User.bIsFrontEndPreview {type=NiagaraBool}
  User.bPawnLanded {type=NiagaraBool}
  User.bPawnSprintOrSlide {type=NiagaraBool}
 ]
}
