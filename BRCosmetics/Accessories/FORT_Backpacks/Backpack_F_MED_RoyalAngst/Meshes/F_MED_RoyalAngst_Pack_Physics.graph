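# Physics asset for the RoyalAngst backpack cosmetic, defining the bodies that let its necklaces and pendant chain dangle.
# One large capsule covers the main bag, and small base capsules anchor each strand to it through angularly free constraints.
# From those anchors hang a one-link necklace, a two-link necklace and a five-link pendant chain.
# Dynamic links use narrow swing and near-zero twist limits, with the necklaces tighter on the second swing axis.
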PhysicsAsset /BRCosmetics/Accessories/FORT_Backpacks/Backpack_F_MED_RoyalAngst/Meshes/F_MED_RoyalAngst_Pack_Physics
bodies {
 base_main_bag [{
   shapeType=Capsule
   length=13.8758
   radius=7.6249
   shapeName=base_main_bag_capsule
   center=<-3.5573 -4.0821 -3.2757>
   rotation=<rot -0.0002 -1.279 90>
  }]
 base_necklace_a [{shapeType=Capsule, length=0.505, radius=0.505, shapeName=base_necklace_a_capsule, rotation=<rot 0 -90 0>}]
 base_necklace_b_01 [{shapeType=Capsule, length=0.505, radius=0.505, shapeName=base_necklace_b_01_capsule, rotation=<rot 0 -90 0>}]
 base_pendant_chan_01 [{shapeType=Capsule, length=0.505, radius=0.505, shapeName=base_pendant_chan_01_capsule, rotation=<rot 0 -90 0>}]
 dyn_necklace_a [{shapeType=Capsule, length=2, radius=1, shapeName=dyn_necklace_a_capsule, center=<2 0 0>, rotation=<rot 90 0 0>}]
 dyn_necklace_b_01 [{shapeType=Capsule, length=2, radius=1, shapeName=dyn_necklace_b_01_capsule, center=<2 0 0>, rotation=<rot 90 0 0>}]
 dyn_necklace_b_02 [{shapeType=Capsule, length=2, radius=1, shapeName=dyn_necklace_b_02_capsule, center=<2 0.3077 -2.915>, rotation=<rot 90 0 0>}]
 dyn_pendant_chan_01 [{shapeType=Capsule, length=2, radius=1, shapeName=dyn_pendant_chan_01_capsule, center=<2.0041 -0.5794 0>, rotation=<rot 90 169.3803 161.0763>}]
 dyn_pendant_chan_02 [{shapeType=Capsule, length=2, radius=1, shapeName=dyn_pendant_chan_02_capsule, center=<2 -0.5063 0>, rotation=<rot 90 -24.7078 -26.9127>}]
 dyn_pendant_chan_03 [{shapeType=Capsule, length=2, radius=1, shapeName=dyn_pendant_chan_03_capsule, center=<2.1901 -0.2634 0>, rotation=<rot 90 0 -11.7268>}]
 dyn_pendant_chan_04 [{shapeType=Capsule, length=2, radius=1, shapeName=dyn_pendant_chan_04_capsule, center=<2 -0.347 0>, rotation=<rot 90 -30.9809 -34.8458>}]
 dyn_pendant_chan_05 [{shapeType=Capsule, length=2, radius=1, shapeName=dyn_pendant_chan_05_capsule, center=<1.9843 0.1006 0>, rotation=<rot 90 0 -8.8999>}]
}
constraints [
 {bone1Name=dyn_pendant_chan_01, bone2Name=base_pendant_chan_01, swing1LimitDegrees=25, swing2LimitDegrees=20, twistLimitDegrees=1}
 {bone1Name=dyn_pendant_chan_02, bone2Name=dyn_pendant_chan_01, swing1LimitDegrees=25, swing2LimitDegrees=20, twistLimitDegrees=1}
 {bone1Name=dyn_pendant_chan_03, bone2Name=dyn_pendant_chan_02, swing1LimitDegrees=25, swing2LimitDegrees=20, twistLimitDegrees=1}
 {bone1Name=dyn_pendant_chan_04, bone2Name=dyn_pendant_chan_03, swing1LimitDegrees=25, swing2LimitDegrees=20, twistLimitDegrees=1}
 {bone1Name=dyn_pendant_chan_05, bone2Name=dyn_pendant_chan_04, swing1LimitDegrees=25, swing2LimitDegrees=20, twistLimitDegrees=1}
 {
  bone1Name=base_pendant_chan_01
  bone2Name=base_main_bag
  swing1LimitDegrees=45
  swing1Motion=Free
  swing2LimitDegrees=45
  swing2Motion=Free
  twistLimitDegrees=45
  twistMotion=Free
 }
 {
  bone1Name=base_necklace_a
  bone2Name=base_main_bag
  swing1LimitDegrees=45
  swing1Motion=Free
  swing2LimitDegrees=45
  swing2Motion=Free
  twistLimitDegrees=45
  twistMotion=Free
 }
 {bone1Name=dyn_necklace_a, bone2Name=base_necklace_a, swing1LimitDegrees=25, swing2LimitDegrees=5, twistLimitDegrees=1}
 {
  bone1Name=base_necklace_b_01
  bone2Name=base_main_bag
  swing1LimitDegrees=45
  swing1Motion=Free
  swing2LimitDegrees=45
  swing2Motion=Free
  twistLimitDegrees=45
  twistMotion=Free
 }
 {bone1Name=dyn_necklace_b_01, bone2Name=base_necklace_b_01, swing1LimitDegrees=25, swing2LimitDegrees=5, twistLimitDegrees=1}
 {bone1Name=dyn_necklace_b_02, bone2Name=dyn_necklace_b_01, swing1LimitDegrees=25, swing2LimitDegrees=2, twistLimitDegrees=1}
]
